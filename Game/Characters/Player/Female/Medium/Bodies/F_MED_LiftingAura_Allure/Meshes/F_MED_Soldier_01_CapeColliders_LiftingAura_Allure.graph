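# Physics asset for the female medium LiftingAura Allure body, named as cape colliders and built entirely from capsules.
# Hand-fitted capsules cover pelvis, spine, clavicles, head and feet, while arm, thigh, calf and neck bodies carry no shapes.
# Dynamic skirt bone chains hang from pelvis and thighs with mostly uniform capsules, branching heavily on the back left.
# Constraints follow these chains with all swing and twist motion free.
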PhysicsAsset /Game/Characters/Player/Female/Medium/Bodies/F_MED_LiftingAura_Allure/Meshes/F_MED_Soldier_01_CapeColliders_LiftingAura_Allure
bodies {
 calf_l []
 calf_r []
 clavicle_l [{shapeType=Capsule, length=6.5623, radius=7.5194, center=<3.9566 -1.0862 -2.1127>, rotation=<rot -77.5566 177.3915 7.7314>}]
 clavicle_r [{shapeType=Capsule, length=6.5623, radius=7.5194, center=<-3.9566 1.0862 2.1127>, rotation=<rot -77.5566 177.3915 -172.2686>}]
 dyn_skirtBack_01 [{shapeType=Capsule, length=5, radius=10, shapeName=dyn_skirtBack_01_capsule, rotation=<rot 90 0 0>}]
 dyn_skirtBack_01_l [{shapeType=Capsule, length=5, radius=10, shapeName=dyn_skirtBack_01_l_capsule, rotation=<rot 90 0 0>}]
 dyn_skirtBack_01_r [{shapeType=Capsule, length=5, radius=10, shapeName=dyn_skirtBack_01_r_capsule, rotation=<rot 90 0 0>}]
 dyn_skirtBack_02 [{shapeType=Capsule, length=5, radius=10, shapeName=dyn_skirtBack_02_capsule, rotation=<rot 90 0 0>}]
 dyn_skirtBack_02_l [{shapeType=Capsule, length=5, radius=10, shapeName=dyn_skirtBack_02_l_capsule, rotation=<rot 90 0 0>}]
 dyn_skirtBack_02_r [{shapeType=Capsule, length=5, radius=10, shapeName=dyn_skirtBack_02_r_capsule, rotation=<rot 90 0 0>}]
 dyn_skirtBack_03 [{shapeType=Capsule, length=5, radius=10, shapeName=dyn_skirtBack_03_capsule, rotation=<rot 90 0 0>}]
 dyn_skirtBack_03_a [{shapeType=Capsule, length=1.5227, radius=15.1845, shapeName=dyn_skirtBack_03_a_capsule, rotation=<rot 90 0 0>}]
 dyn_skirtBack_03_l [{shapeType=Capsule, length=5, radius=10, shapeName=dyn_skirtBack_03_l_capsule, rotation=<rot 90 0 0>}]
 dyn_skirtBack_03_r [{shapeType=Capsule, length=5, radius=10, shapeName=dyn_skirtBack_03_r_capsule, rotation=<rot 90 0 0>}]
 dyn_skirtBack_04 [{shapeType=Capsule, length=5, radius=10, shapeName=dyn_skirtBack_04_capsule, rotation=<rot 90 0 0>}]
 dyn_skirtBack_04_a [{shapeType=Capsule, length=1.5227, radius=15.1845, shapeName=dyn_skirtBack_04_a_capsule, rotation=<rot 90 0 0>}]
 dyn_skirtBack_04_a_02 [{shapeType=Capsule, length=1.5227, radius=15.1845, shapeName=dyn_skirtBack_04_a_02_capsule, rotation=<rot 90 0 0>}]
 dyn_skirtBack_04_a_MSTR [{shapeType=Capsule, length=5, radius=10, shapeName=dyn_skirtBack_04_a_MSTR_capsule, rotation=<rot 90 0 0>}]
 dyn_skirtBack_04_l [{shapeType=Capsule, length=5, radius=10, shapeName=dyn_skirtBack_04_l_capsule, rotation=<rot 90 0 0>}]
 dyn_skirtBack_04_r [{shapeType=Capsule, length=8.0732, radius=14.2848, shapeName=dyn_skirtBack_04_r_capsule, rotation=<rot 90 0 0>}]
 dyn_skirtBack_05_l [{shapeType=Capsule, length=5, radius=10, shapeName=dyn_skirtBack_05_l_capsule, rotation=<rot 90 0 0>}]
 dyn_skirtBack_05_l_MSTR [{shapeType=Capsule, length=5, radius=10, shapeName=dyn_skirtBack_05_l_MSTR_capsule, rotation=<rot 90 0 0>}]
 dyn_skirtBack_05_r [{shapeType=Capsule, length=4.1313, radius=18.2266, shapeName=dyn_skirtBack_05_r_capsule, rotation=<rot 90 0 0>}]
 dyn_skirtBack_06_l [{shapeType=Capsule, length=5, radius=10, shapeName=dyn_skirtBack_06_l_capsule, rotation=<rot 90 0 0>}]
 dyn_skirtBack_06_r [{shapeType=Capsule, length=2.4032, radius=19.9547, shapeName=dyn_skirtBack_06_r_capsule, rotation=<rot 90 0 0>}]
 dyn_skirtBack_06_r3 [{shapeType=Capsule, length=5, radius=10, shapeName=dyn_skirtBack_06_r3_capsule, rotation=<rot 90 0 0>}]
 dyn_skirtBack_l_D_MSTR [{shapeType=Capsule, length=5, radius=10, shapeName=dyn_skirtBack_l_D_MSTR_capsule, rotation=<rot 90 0 0>}]
 dyn_skirtBack_l_a_01 [{shapeType=Capsule, length=5, radius=10, shapeName=dyn_skirtBack_l_a_01_capsule, rotation=<rot 90 0 0>}]
 dyn_skirtBack_l_a_02 [{shapeType=Capsule, length=5, radius=10, shapeName=dyn_skirtBack_l_a_02_capsule, rotation=<rot 90 0 0>}]
 dyn_skirtBack_l_a_MSTR [{shapeType=Capsule, length=5, radius=10, shapeName=dyn_skirtBack_l_a_MSTR_capsule, rotation=<rot 90 0 0>}]
 dyn_skirtBack_l_b_01 [{shapeType=Capsule, length=5, radius=10, shapeName=dyn_skirtBack_l_b_01_capsule, rotation=<rot 90 0 0>}]
 dyn_skirtBack_l_b_02 [{shapeType=Capsule, length=5, radius=10, shapeName=dyn_skirtBack_l_b_02_capsule, rotation=<rot 90 0 0>}]
 dyn_skirtBack_l_b_03 [{shapeType=Capsule, length=5, radius=10, shapeName=dyn_skirtBack_l_b_03_capsule, rotation=<rot 90 0 0>}]
 dyn_skirtBack_l_b_MSTR [{shapeType=Capsule, length=5, radius=10, shapeName=dyn_skirtBack_l_b_MSTR_capsule, rotation=<rot 90 0 0>}]
 dyn_skirtBack_l_c_01 [{shapeType=Capsule, length=5, radius=10, shapeName=dyn_skirtBack_l_c_01_capsule, rotation=<rot 90 0 0>}]
 dyn_skirtBack_l_c_02 [{shapeType=Capsule, length=5, radius=10, shapeName=dyn_skirtBack_l_c_02_capsule, rotation=<rot 90 0 0>}]
 dyn_skirtBack_l_c_03 [{shapeType=Capsule, length=5, radius=10, shapeName=dyn_skirtBack_l_c_03_capsule, rotation=<rot 90 0 0>}]
 dyn_skirtBack_l_c_MSTR [{shapeType=Capsule, length=5, radius=10, shapeName=dyn_skirtBack_l_c_MSTR_capsule, rotation=<rot 90 0 0>}]
 dyn_skirtBack_l_d_01 [{shapeType=Capsule, length=5, radius=10, shapeName=dyn_skirtBack_l_d_01_capsule, rotation=<rot 90 0 0>}]
 dyn_skirtBack_l_flyoff_01 [{shapeType=Capsule, length=5, radius=10, shapeName=dyn_skirtBack_l_flyoff_01_capsule, rotation=<rot 90 0 0>}]
 dyn_skirtBack_l_flyoff_02 [{shapeType=Capsule, length=5, radius=10, shapeName=dyn_skirtBack_l_flyoff_02_capsule, rotation=<rot 90 0 0>}]
 dyn_skirtBack_l_flyoff_03 [{shapeType=Capsule, length=5, radius=10, shapeName=dyn_skirtBack_l_flyoff_03_capsule, rotation=<rot 90 0 0>}]
 dyn_skirtBack_l_flyoff_04 [{shapeType=Capsule, length=5, radius=10, shapeName=dyn_skirtBack_l_flyoff_04_capsule, rotation=<rot 90 0 0>}]
 dyn_skirtBack_l_flyoff_MSTR [{shapeType=Capsule, length=5, radius=10, shapeName=dyn_skirtBack_l_flyoff_MSTR_capsule, rotation=<rot 90 0 0>}]
 dyn_skirtBack_mstr [{shapeType=Capsule, length=5, radius=10, shapeName=dyn_skirtBack_mstr_capsule, rotation=<rot 90 0 0>}]
 dyn_skirtBack_mstr_l [{shapeType=Capsule, length=5, radius=10, shapeName=dyn_skirtBack_mstr_l_capsule, rotation=<rot 90 0 0>}]
 dyn_skirtBack_mstr_r [{shapeType=Capsule, length=5, radius=10, shapeName=dyn_skirtBack_mstr_r_capsule, rotation=<rot 90 0 0>}]
 dyn_skirtSideFront_l_01 [{shapeType=Capsule, length=5, radius=10, shapeName=dyn_skirtSideFront_l_01_capsule, rotation=<rot 90 0 0>}]
 dyn_skirtSideFront_l_02 [{shapeType=Capsule, length=5, radius=10, shapeName=dyn_skirtSideFront_l_02_capsule, rotation=<rot 90 0 0>}]
 dyn_skirtSideFront_l_03 [{shapeType=Capsule, length=5, radius=10, shapeName=dyn_skirtSideFront_l_03_capsule, rotation=<rot 90 0 0>}]
 dyn_skirtSideFront_l_MSTR [{shapeType=Capsule, length=5, radius=10, shapeName=dyn_skirtSideFront_l_MSTR_capsule, rotation=<rot 90 0 0>}]
 dyn_skirtSideFront_r_01 [{shapeType=Capsule, length=5, radius=10, shapeName=dyn_skirtSideFront_r_01_capsule, rotation=<rot 90 0 0>}]
 dyn_skirtSideFront_r_02 [{shapeType=Capsule, length=5, radius=10, shapeName=dyn_skirtSideFront_r_02_capsule, rotation=<rot 90 0 0>}]
 dyn_skirtSideFront_r_03 [{shapeType=Capsule, length=5, radius=10, shapeName=dyn_skirtSideFront_r_03_capsule, rotation=<rot 90 0 0>}]
 dyn_skirtSideFront_r_MSTR [{shapeType=Capsule, length=5, radius=10, shapeName=dyn_skirtSideFront_r_MSTR_capsule, rotation=<rot 90 0 0>}]
 dyn_skirtSide_01_l [{shapeType=Capsule, length=5, radius=10, shapeName=dyn_skirtSide_01_l_capsule, rotation=<rot 90 0 0>}]
 dyn_skirtSide_01_r [{shapeType=Capsule, length=5, radius=10, shapeName=dyn_skirtSide_01_r_capsule, rotation=<rot 90 0 0>}]
 dyn_skirtSide_02_l [{shapeType=Capsule, length=5, radius=10, shapeName=dyn_skirtSide_02_l_capsule, rotation=<rot 90 0 0>}]
 dyn_skirtSide_02_r [{shapeType=Capsule, length=5, radius=10, shapeName=dyn_skirtSide_02_r_capsule, rotation=<rot 90 0 0>}]
 dyn_skirtSide_03_l [{shapeType=Capsule, length=5, radius=10, shapeName=dyn_skirtSide_03_l_capsule, rotation=<rot 90 0 0>}]
 dyn_skirtSide_03_r [{shapeType=Capsule, length=5, radius=10, shapeName=dyn_skirtSide_03_r_capsule, rotation=<rot 90 0 0>}]
 dyn_skirtSide_04_l [{shapeType=Capsule, length=5, radius=10, shapeName=dyn_skirtSide_04_l_capsule, rotation=<rot 90 0 0>}]
 dyn_skirtSide_04_r [{shapeType=Capsule, length=5, radius=10, shapeName=dyn_skirtSide_04_r_capsule, rotation=<rot 90 0 0>}]
 dyn_skirtSide_05_r [{shapeType=Capsule, length=5, radius=10, shapeName=dyn_skirtSide_05_r_capsule, rotation=<rot 90 0 0>}]
 dyn_skirtSide_mstr_l [{shapeType=Capsule, length=5, radius=10, shapeName=dyn_skirtSide_mstr_l_capsule, rotation=<rot 90 0 0>}]
 dyn_skirtSide_mstr_r [{shapeType=Capsule, length=5, radius=10, shapeName=dyn_skirtSide_mstr_r_capsule, rotation=<rot 90 0 0>}]
 foot_l [{shapeType=Capsule, length=11.0765, radius=8.0344, center=<-4.1811 6.0027 -0.338>, rotation=<rot 0.0388 1.4701 -89.3982>}]
 foot_r [{shapeType=Capsule, length=11.0765, radius=8.0344, center=<4.1811 -6.0027 0.338>, rotation=<rot 0.0388 1.4701 90.6018>}]
 head [{shapeType=Capsule, length=0, radius=18, center=<9.0697 0.0746 0>, rotation=<rot 0 -90 0>}]
 lowerarm_l []
 lowerarm_r []
 neck_02 []
 pelvis [{shapeType=Capsule, length=9.6042, radius=11.1441, center=<-2.4471 -2.5289 -0.1981>}]
 spine_01 [{shapeType=Capsule, length=4.3888, radius=14.3638, center=<-8.9021 -0.8973 0>}]
 spine_02 [
  {shapeType=Capsule, length=2.5219, radius=9.7317, center=<-2.6885 -1.2641 0>}
  {shapeType=Capsule, length=20.6875, radius=9.3125, center=<-1.8317 1.9362 0>, rotation=<rot 90 0 0>}
 ]
 spine_03 [
  {shapeType=Capsule, length=6.1863, radius=8.4854, center=<0.0574 0.4416 0>}
  {shapeType=Capsule, length=20.7188, radius=9.2812, center=<-1.2163 0.549 0>, rotation=<rot 90 180 -175.0732>}
 ]
 spine_04 [
  {shapeType=Capsule, length=9.0076, radius=10.3916, center=<5.6234 -0.6779 0>}
  {shapeType=Capsule, length=10.0312, radius=9.2812, center=<-1.2326 1.308 0>, rotation=<rot 90 0 0>}
 ]
 spine_05 [{shapeType=Capsule, length=9.678, radius=12.3571, center=<0.6969 1.91 0>}]
 thigh_l []
 thigh_r []
 upperarm_l []
 upperarm_r []
}
constraints [
 {
  bone1Name=spine_01
  bone2Name=pelvis
  swing1LimitDegrees=45
  swing1Motion=Free
  swing2LimitDegrees=45
  swing2Motion=Free
  twistLimitDegrees=45
  twistMotion=Free
 }
 {
  bone1Name=spine_02
  bone2Name=spine_01
  swing1LimitDegrees=45
  swing1Motion=Free
  swing2LimitDegrees=45
  swing2Motion=Free
  twistLimitDegrees=45
  twistMotion=Free
 }
 {
  bone1Name=spine_03
  bone2Name=spine_02
  swing1LimitDegrees=45
  swing1Motion=Free
  swing2LimitDegrees=45
  swing2Motion=Free
  twistLimitDegrees=45
  twistMotion=Free
 }
 {
  bone1Name=spine_04
  bone2Name=spine_03
  swing1LimitDegrees=45
  swing1Motion=Free
  swing2LimitDegrees=45
  swing2Motion=Free
  twistLimitDegrees=45
  twistMotion=Free
 }
 {
  bone1Name=spine_05
  bone2Name=spine_04
  swing1LimitDegrees=45
  swing1Motion=Free
  swing2LimitDegrees=45
  swing2Motion=Free
  twistLimitDegrees=45
  twistMotion=Free
 }
 {
  bone1Name=clavicle_l
  bone2Name=spine_05
  swing1LimitDegrees=45
  swing1Motion=Free
  swing2LimitDegrees=45
  swing2Motion=Free
  twistLimitDegrees=45
  twistMotion=Free
 }
 {
  bone1Name=upperarm_l
  bone2Name=clavicle_l
  swing1LimitDegrees=45
  swing1Motion=Free
  swing2LimitDegrees=45
  swing2Motion=Free
  twistLimitDegrees=45
  twistMotion=Free
 }
 {
  bone1Name=lowerarm_l
  bone2Name=upperarm_l
  swing1LimitDegrees=45
  swing1Motion=Free
  swing2LimitDegrees=45
  swing2Motion=Free
  twistLimitDegrees=45
  twistMotion=Free
 }
 {
  bone1Name=neck_02
  bone2Name=spine_05
  swing1LimitDegrees=45
  swing1Motion=Free
  swing2LimitDegrees=45
  swing2Motion=Free
  twistLimitDegrees=45
  twistMotion=Free
 }
 {
  bone1Name=head
  bone2Name=neck_02
  swing1LimitDegrees=45
  swing1Motion=Free
  swing2LimitDegrees=45
  swing2Motion=Free
  twistLimitDegrees=45
  twistMotion=Free
 }
 {
  bone1Name=calf_l
  bone2Name=thigh_l
  swing1LimitDegrees=45
  swing1Motion=Free
  swing2LimitDegrees=45
  swing2Motion=Free
  twistLimitDegrees=45
  twistMotion=Free
 }
 {
  bone1Name=thigh_l
  bone2Name=pelvis
  swing1LimitDegrees=45
  swing1Motion=Free
  swing2LimitDegrees=45
  swing2Motion=Free
  twistLimitDegrees=45
  twistMotion=Free
 }
 {
  bone1Name=foot_l
  bone2Name=calf_l
  swing1LimitDegrees=45
  swing1Motion=Free
  swing2LimitDegrees=45
  swing2Motion=Free
  twistLimitDegrees=45
  twistMotion=Free
 }
 {
  bone1Name=lowerarm_r
  bone2Name=upperarm_r
  swing1LimitDegrees=45
  swing1Motion=Free
  swing2LimitDegrees=45
  swing2Motion=Free
  twistLimitDegrees=45
  twistMotion=Free
 }
 {
  bone1Name=upperarm_r
  bone2Name=clavicle_r
  swing1LimitDegrees=45
  swing1Motion=Free
  swing2LimitDegrees=45
  swing2Motion=Free
  twistLimitDegrees=45
  twistMotion=Free
 }
 {
  bone1Name=clavicle_r
  bone2Name=spine_05
  swing1LimitDegrees=45
  swing1Motion=Free
  swing2LimitDegrees=45
  swing2Motion=Free
  twistLimitDegrees=45
  twistMotion=Free
 }
 {
  bone1Name=thigh_r
  bone2Name=pelvis
  swing1LimitDegrees=45
  swing1Motion=Free
  swing2LimitDegrees=45
  swing2Motion=Free
  twistLimitDegrees=45
  twistMotion=Free
 }
 {
  bone1Name=foot_r
  bone2Name=calf_r
  swing1LimitDegrees=45
  swing1Motion=Free
  swing2LimitDegrees=45
  swing2Motion=Free
  twistLimitDegrees=45
  twistMotion=Free
 }
 {
  bone1Name=calf_r
  bone2Name=thigh_r
  swing1LimitDegrees=45
  swing1Motion=Free
  swing2LimitDegrees=45
  swing2Motion=Free
  twistLimitDegrees=45
  twistMotion=Free
 }
 {
  bone1Name=dyn_skirtBack_mstr
  bone2Name=pelvis
  swing1LimitDegrees=45
  swing1Motion=Free
  swing2LimitDegrees=45
  swing2Motion=Free
  twistLimitDegrees=45
  twistMotion=Free
 }
 {
  bone1Name=dyn_skirtBack_01
  bone2Name=dyn_skirtBack_mstr
  swing1LimitDegrees=45
  swing1Motion=Free
  swing2LimitDegrees=45
  swing2Motion=Free
  twistLimitDegrees=45
  twistMotion=Free
 }
 {
  bone1Name=dyn_skirtBack_02
  bone2Name=dyn_skirtBack_01
  swing1LimitDegrees=45
  swing1Motion=Free
  swing2LimitDegrees=45
  swing2Motion=Free
  twistLimitDegrees=45
  twistMotion=Free
 }
 {
  bone1Name=dyn_skirtBack_03
  bone2Name=dyn_skirtBack_02
  swing1LimitDegrees=45
  swing1Motion=Free
  swing2LimitDegrees=45
  swing2Motion=Free
  twistLimitDegrees=45
  twistMotion=Free
 }
 {
  bone1Name=dyn_skirtBack_04
  bone2Name=dyn_skirtBack_03
  swing1LimitDegrees=45
  swing1Motion=Free
  swing2LimitDegrees=45
  swing2Motion=Free
  twistLimitDegrees=45
  twistMotion=Free
 }
 {
  bone1Name=dyn_skirtBack_mstr_l
  bone2Name=pelvis
  swing1LimitDegrees=45
  swing1Motion=Free
  swing2LimitDegrees=45
  swing2Motion=Free
  twistLimitDegrees=45
  twistMotion=Free
 }
 {
  bone1Name=dyn_skirtBack_01_l
  bone2Name=dyn_skirtBack_mstr_l
  swing1LimitDegrees=45
  swing1Motion=Free
  swing2LimitDegrees=45
  swing2Motion=Free
  twistLimitDegrees=45
  twistMotion=Free
 }
 {
  bone1Name=dyn_skirtBack_02_l
  bone2Name=dyn_skirtBack_01_l
  swing1LimitDegrees=45
  swing1Motion=Free
  swing2LimitDegrees=45
  swing2Motion=Free
  twistLimitDegrees=45
  twistMotion=Free
 }
 {
  bone1Name=dyn_skirtBack_03_l
  bone2Name=dyn_skirtBack_02_l
  swing1LimitDegrees=45
  swing1Motion=Free
  swing2LimitDegrees=45
  swing2Motion=Free
  twistLimitDegrees=45
  twistMotion=Free
 }
 {
  bone1Name=dyn_skirtBack_04_a_MSTR
  bone2Name=dyn_skirtBack_03_l
  swing1LimitDegrees=45
  swing1Motion=Free
  swing2LimitDegrees=45
  swing2Motion=Free
  twistLimitDegrees=45
  twistMotion=Free
 }
 {
  bone1Name=dyn_skirtBack_04_a
  bone2Name=dyn_skirtBack_04_a_MSTR
  swing1LimitDegrees=45
  swing1Motion=Free
  swing2LimitDegrees=45
  swing2Motion=Free
  twistLimitDegrees=45
  twistMotion=Free
 }
 {
  bone1Name=dyn_skirtBack_04_a_02
  bone2Name=dyn_skirtBack_04_a
  swing1LimitDegrees=45
  swing1Motion=Free
  swing2LimitDegrees=45
  swing2Motion=Free
  twistLimitDegrees=45
  twistMotion=Free
 }
 {
  bone1Name=dyn_skirtBack_03_a
  bone2Name=dyn_skirtBack_04_a_02
  swing1LimitDegrees=45
  swing1Motion=Free
  swing2LimitDegrees=45
  swing2Motion=Free
  twistLimitDegrees=45
  twistMotion=Free
 }
 {
  bone1Name=dyn_skirtBack_04_l
  bone2Name=dyn_skirtBack_03_l
  swing1LimitDegrees=45
  swing1Motion=Free
  swing2LimitDegrees=45
  swing2Motion=Free
  twistLimitDegrees=45
  twistMotion=Free
 }
 {
  bone1Name=dyn_skirtBack_05_l_MSTR
  bone2Name=dyn_skirtBack_04_l
  swing1LimitDegrees=45
  swing1Motion=Free
  swing2LimitDegrees=45
  swing2Motion=Free
  twistLimitDegrees=45
  twistMotion=Free
 }
 {
  bone1Name=dyn_skirtBack_05_l
  bone2Name=dyn_skirtBack_05_l_MSTR
  swing1LimitDegrees=45
  swing1Motion=Free
  swing2LimitDegrees=45
  swing2Motion=Free
  twistLimitDegrees=45
  twistMotion=Free
 }
 {
  bone1Name=dyn_skirtBack_06_l
  bone2Name=dyn_skirtBack_05_l
  swing1LimitDegrees=45
  swing1Motion=Free
  swing2LimitDegrees=45
  swing2Motion=Free
  twistLimitDegrees=45
  twistMotion=Free
 }
 {
  bone1Name=dyn_skirtBack_06_r3
  bone2Name=dyn_skirtBack_06_l
  swing1LimitDegrees=45
  swing1Motion=Free
  swing2LimitDegrees=45
  swing2Motion=Free
  twistLimitDegrees=45
  twistMotion=Free
 }
 {
  bone1Name=dyn_skirtBack_l_D_MSTR
  bone2Name=dyn_skirtBack_04_l
  swing1LimitDegrees=45
  swing1Motion=Free
  swing2LimitDegrees=45
  swing2Motion=Free
  twistLimitDegrees=45
  twistMotion=Free
 }
 {
  bone1Name=dyn_skirtBack_l_d_01
  bone2Name=dyn_skirtBack_l_D_MSTR
  swing1LimitDegrees=45
  swing1Motion=Free
  swing2LimitDegrees=45
  swing2Motion=Free
  twistLimitDegrees=45
  twistMotion=Free
 }
 {
  bone1Name=dyn_skirtBack_l_a_MSTR
  bone2Name=dyn_skirtBack_04_l
  swing1LimitDegrees=45
  swing1Motion=Free
  swing2LimitDegrees=45
  swing2Motion=Free
  twistLimitDegrees=45
  twistMotion=Free
 }
 {
  bone1Name=dyn_skirtBack_l_a_01
  bone2Name=dyn_skirtBack_l_a_MSTR
  swing1LimitDegrees=45
  swing1Motion=Free
  swing2LimitDegrees=45
  swing2Motion=Free
  twistLimitDegrees=45
  twistMotion=Free
 }
 {
  bone1Name=dyn_skirtBack_l_a_02
  bone2Name=dyn_skirtBack_l_a_01
  swing1LimitDegrees=45
  swing1Motion=Free
  swing2LimitDegrees=45
  swing2Motion=Free
  twistLimitDegrees=45
  twistMotion=Free
 }
 {
  bone1Name=dyn_skirtBack_l_b_MSTR
  bone2Name=dyn_skirtBack_04_l
  swing1LimitDegrees=45
  swing1Motion=Free
  swing2LimitDegrees=45
  swing2Motion=Free
  twistLimitDegrees=45
  twistMotion=Free
 }
 {
  bone1Name=dyn_skirtBack_l_b_01
  bone2Name=dyn_skirtBack_l_b_MSTR
  swing1LimitDegrees=45
  swing1Motion=Free
  swing2LimitDegrees=45
  swing2Motion=Free
  twistLimitDegrees=45
  twistMotion=Free
 }
 {
  bone1Name=dyn_skirtBack_l_b_02
  bone2Name=dyn_skirtBack_l_b_01
  swing1LimitDegrees=45
  swing1Motion=Free
  swing2LimitDegrees=45
  swing2Motion=Free
  twistLimitDegrees=45
  twistMotion=Free
 }
 {
  bone1Name=dyn_skirtBack_l_b_03
  bone2Name=dyn_skirtBack_l_b_02
  swing1LimitDegrees=45
  swing1Motion=Free
  swing2LimitDegrees=45
  swing2Motion=Free
  twistLimitDegrees=45
  twistMotion=Free
 }
 {
  bone1Name=dyn_skirtBack_l_c_MSTR
  bone2Name=dyn_skirtBack_03_l
  swing1LimitDegrees=45
  swing1Motion=Free
  swing2LimitDegrees=45
  swing2Motion=Free
  twistLimitDegrees=45
  twistMotion=Free
 }
 {
  bone1Name=dyn_skirtBack_l_c_01
  bone2Name=dyn_skirtBack_l_c_MSTR
  swing1LimitDegrees=45
  swing1Motion=Free
  swing2LimitDegrees=45
  swing2Motion=Free
  twistLimitDegrees=45
  twistMotion=Free
 }
 {
  bone1Name=dyn_skirtBack_l_c_02
  bone2Name=dyn_skirtBack_l_c_01
  swing1LimitDegrees=45
  swing1Motion=Free
  swing2LimitDegrees=45
  swing2Motion=Free
  twistLimitDegrees=45
  twistMotion=Free
 }
 {
  bone1Name=dyn_skirtBack_l_c_03
  bone2Name=dyn_skirtBack_l_c_02
  swing1LimitDegrees=45
  swing1Motion=Free
  swing2LimitDegrees=45
  swing2Motion=Free
  twistLimitDegrees=45
  twistMotion=Free
 }
 {
  bone1Name=dyn_skirtBack_mstr_r
  bone2Name=pelvis
  swing1LimitDegrees=45
  swing1Motion=Free
  swing2LimitDegrees=45
  swing2Motion=Free
  twistLimitDegrees=45
  twistMotion=Free
 }
 {
  bone1Name=dyn_skirtBack_01_r
  bone2Name=dyn_skirtBack_mstr_r
  swing1LimitDegrees=45
  swing1Motion=Free
  swing2LimitDegrees=45
  swing2Motion=Free
  twistLimitDegrees=45
  twistMotion=Free
 }
 {
  bone1Name=dyn_skirtBack_02_r
  bone2Name=dyn_skirtBack_01_r
  swing1LimitDegrees=45
  swing1Motion=Free
  swing2LimitDegrees=45
  swing2Motion=Free
  twistLimitDegrees=45
  twistMotion=Free
 }
 {
  bone1Name=dyn_skirtBack_03_r
  bone2Name=dyn_skirtBack_02_r
  swing1LimitDegrees=45
  swing1Motion=Free
  swing2LimitDegrees=45
  swing2Motion=Free
  twistLimitDegrees=45
  twistMotion=Free
 }
 {
  bone1Name=dyn_skirtBack_04_r
  bone2Name=dyn_skirtBack_03_r
  swing1LimitDegrees=45
  swing1Motion=Free
  swing2LimitDegrees=45
  swing2Motion=Free
  twistLimitDegrees=45
  twistMotion=Free
 }
 {
  bone1Name=dyn_skirtBack_05_r
  bone2Name=dyn_skirtBack_04_r
  swing1LimitDegrees=45
  swing1Motion=Free
  swing2LimitDegrees=45
  swing2Motion=Free
  twistLimitDegrees=45
  twistMotion=Free
 }
 {
  bone1Name=dyn_skirtBack_06_r
  bone2Name=dyn_skirtBack_05_r
  swing1LimitDegrees=45
  swing1Motion=Free
  swing2LimitDegrees=45
  swing2Motion=Free
  twistLimitDegrees=45
  twistMotion=Free
 }
 {
  bone1Name=dyn_skirtSide_mstr_r
  bone2Name=pelvis
  swing1LimitDegrees=45
  swing1Motion=Free
  swing2LimitDegrees=45
  swing2Motion=Free
  twistLimitDegrees=45
  twistMotion=Free
 }
 {
  bone1Name=dyn_skirtSide_01_r
  bone2Name=dyn_skirtSide_mstr_r
  swing1LimitDegrees=45
  swing1Motion=Free
  swing2LimitDegrees=45
  swing2Motion=Free
  twistLimitDegrees=45
  twistMotion=Free
 }
 {
  bone1Name=dyn_skirtSide_02_r
  bone2Name=dyn_skirtSide_01_r
  swing1LimitDegrees=45
  swing1Motion=Free
  swing2LimitDegrees=45
  swing2Motion=Free
  twistLimitDegrees=45
  twistMotion=Free
 }
 {
  bone1Name=dyn_skirtSide_03_r
  bone2Name=dyn_skirtSide_02_r
  swing1LimitDegrees=45
  swing1Motion=Free
  swing2LimitDegrees=45
  swing2Motion=Free
  twistLimitDegrees=45
  twistMotion=Free
 }
 {
  bone1Name=dyn_skirtSide_04_r
  bone2Name=dyn_skirtSide_03_r
  swing1LimitDegrees=45
  swing1Motion=Free
  swing2LimitDegrees=45
  swing2Motion=Free
  twistLimitDegrees=45
  twistMotion=Free
 }
 {
  bone1Name=dyn_skirtSide_05_r
  bone2Name=dyn_skirtSide_04_r
  swing1LimitDegrees=45
  swing1Motion=Free
  swing2LimitDegrees=45
  swing2Motion=Free
  twistLimitDegrees=45
  twistMotion=Free
 }
 {
  bone1Name=dyn_skirtBack_l_flyoff_MSTR
  bone2Name=thigh_l
  swing1LimitDegrees=45
  swing1Motion=Free
  swing2LimitDegrees=45
  swing2Motion=Free
  twistLimitDegrees=45
  twistMotion=Free
 }
 {
  bone1Name=dyn_skirtBack_l_flyoff_01
  bone2Name=dyn_skirtBack_l_flyoff_MSTR
  swing1LimitDegrees=45
  swing1Motion=Free
  swing2LimitDegrees=45
  swing2Motion=Free
  twistLimitDegrees=45
  twistMotion=Free
 }
 {
  bone1Name=dyn_skirtBack_l_flyoff_02
  bone2Name=dyn_skirtBack_l_flyoff_01
  swing1LimitDegrees=45
  swing1Motion=Free
  swing2LimitDegrees=45
  swing2Motion=Free
  twistLimitDegrees=45
  twistMotion=Free
 }
 {
  bone1Name=dyn_skirtBack_l_flyoff_03
  bone2Name=dyn_skirtBack_l_flyoff_02
  swing1LimitDegrees=45
  swing1Motion=Free
  swing2LimitDegrees=45
  swing2Motion=Free
  twistLimitDegrees=45
  twistMotion=Free
 }
 {
  bone1Name=dyn_skirtBack_l_flyoff_04
  bone2Name=dyn_skirtBack_l_flyoff_03
  swing1LimitDegrees=45
  swing1Motion=Free
  swing2LimitDegrees=45
  swing2Motion=Free
  twistLimitDegrees=45
  twistMotion=Free
 }
 {
  bone1Name=dyn_skirtSideFront_l_MSTR
  bone2Name=thigh_l
  swing1LimitDegrees=45
  swing1Motion=Free
  swing2LimitDegrees=45
  swing2Motion=Free
  twistLimitDegrees=45
  twistMotion=Free
 }
 {
  bone1Name=dyn_skirtSideFront_l_01
  bone2Name=dyn_skirtSideFront_l_MSTR
  swing1LimitDegrees=45
  swing1Motion=Free
  swing2LimitDegrees=45
  swing2Motion=Free
  twistLimitDegrees=45
  twistMotion=Free
 }
 {
  bone1Name=dyn_skirtSideFront_l_02
  bone2Name=dyn_skirtSideFront_l_01
  swing1LimitDegrees=45
  swing1Motion=Free
  swing2LimitDegrees=45
  swing2Motion=Free
  twistLimitDegrees=45
  twistMotion=Free
 }
 {
  bone1Name=dyn_skirtSideFront_l_03
  bone2Name=dyn_skirtSideFront_l_02
  swing1LimitDegrees=45
  swing1Motion=Free
  swing2LimitDegrees=45
  swing2Motion=Free
  twistLimitDegrees=45
  twistMotion=Free
 }
 {
  bone1Name=dyn_skirtSide_mstr_l
  bone2Name=thigh_l
  swing1LimitDegrees=45
  swing1Motion=Free
  swing2LimitDegrees=45
  swing2Motion=Free
  twistLimitDegrees=45
  twistMotion=Free
 }
 {
  bone1Name=dyn_skirtSide_01_l
  bone2Name=dyn_skirtSide_mstr_l
  swing1LimitDegrees=45
  swing1Motion=Free
  swing2LimitDegrees=45
  swing2Motion=Free
  twistLimitDegrees=45
  twistMotion=Free
 }
 {
  bone1Name=dyn_skirtSide_02_l
  bone2Name=dyn_skirtSide_01_l
  swing1LimitDegrees=45
  swing1Motion=Free
  swing2LimitDegrees=45
  swing2Motion=Free
  twistLimitDegrees=45
  twistMotion=Free
 }
 {
  bone1Name=dyn_skirtSide_03_l
  bone2Name=dyn_skirtSide_02_l
  swing1LimitDegrees=45
  swing1Motion=Free
  swing2LimitDegrees=45
  swing2Motion=Free
  twistLimitDegrees=45
  twistMotion=Free
 }
 {
  bone1Name=dyn_skirtSide_04_l
  bone2Name=dyn_skirtSide_03_l
  swing1LimitDegrees=45
  swing1Motion=Free
  swing2LimitDegrees=45
  swing2Motion=Free
  twistLimitDegrees=45
  twistMotion=Free
 }
 {
  bone1Name=dyn_skirtSideFront_r_MSTR
  bone2Name=thigh_r
  swing1LimitDegrees=45
  swing1Motion=Free
  swing2LimitDegrees=45
  swing2Motion=Free
  twistLimitDegrees=45
  twistMotion=Free
 }
 {
  bone1Name=dyn_skirtSideFront_r_01
  bone2Name=dyn_skirtSideFront_r_MSTR
  swing1LimitDegrees=45
  swing1Motion=Free
  swing2LimitDegrees=45
  swing2Motion=Free
  twistLimitDegrees=45
  twistMotion=Free
 }
 {
  bone1Name=dyn_skirtSideFront_r_02
  bone2Name=dyn_skirtSideFront_r_01
  swing1LimitDegrees=45
  swing1Motion=Free
  swing2LimitDegrees=45
  swing2Motion=Free
  twistLimitDegrees=45
  twistMotion=Free
 }
 {
  bone1Name=dyn_skirtSideFront_r_03
  bone2Name=dyn_skirtSideFront_r_02
  swing1LimitDegrees=45
  swing1Motion=Free
  swing2LimitDegrees=45
  swing2Motion=Free
  twistLimitDegrees=45
  twistMotion=Free
 }
]
